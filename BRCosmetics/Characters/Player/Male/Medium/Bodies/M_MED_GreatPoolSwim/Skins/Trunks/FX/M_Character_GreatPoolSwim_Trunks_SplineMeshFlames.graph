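# Translucent unlit Niagara material drawing wispy flame strands for the GreatPoolSwim Trunks cosmetic.
# SplineThicken sets strand width, while UV-weighted world position offset adds sine wiggle, lift and clamped drag from dynamic-parameter positions.
# Panning noise, offset by vertex color, is shaped by length, concavity and optional swipe masks, driving opacity and a tip-to-particle color blend with time-of-day tint.
# An optional switch isolates meshes by matching vertex green to Mesh Index.
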Material /BRCosmetics/Characters/Player/Male/Medium/Bodies/M_MED_GreatPoolSwim/Skins/Trunks/FX/M_Character_GreatPoolSwim_Trunks_SplineMeshFlames
graph {
 properties {
  bUsedWithNiagaraMeshParticles=true
  bUsedWithNiagaraRibbons=true
  blendMode=BLEND_Translucent
  pixelDepthOffsetMode=PDOM_Legacy
  shadingModel=MSM_Unlit
  usedShadingModels=Unlit
 }
 nodes [
  n1 = ScalarParameter() {defaultValue=10, group=Spline, parameterName=Width Base}
  n2 = DynamicParameter() {paramNames [WidthTipMultiplier, WidthBaseMultiplier, Concave, U Length]}
  n3 = ScalarParameter() {defaultValue=10, group=Spline, parameterName=Width Tip}
  n4 = TextureCoordinate()
  n5 = DynamicParameter() {parameterIndex=2, defaultValue=<rgba 1 0 1 1>, paramNames [Speed Y offset, Swipe, Mesh Index, Param4]}
  n6 = VertexColor()
  n7 = VertexColor()
  n8 = Add(A=Add(A=ComponentMask(None=@n4) {g=true}, B=Multiply(A=@n5.Speed Y offset, B=@n6.G)), B=@n7.R)
  n9 = ScalarParameter() {defaultValue=1, group=WPO, parameterName=WPO Frequency, sliderMax=2}
  n10 = Time() {period=1}
  n11 = Multiply(A=@n7.R, B=ScalarParameter() {defaultValue=1, group=WPO, parameterName=WPO Speed, sliderMax=1})
  n12 = Sine(None=Add(A=Multiply(A=@n8, B=@n9), B=Multiply(A=@n10, B=@n11)))
  n13 = VectorParameter() {group=WPO, parameterName=WPO Tangent Vector, defaultValue=<rgba 1 0 1 1>}
  n14 = Multiply(A=Multiply(A=@n12, B=Transform(None=@n13.RGB)), B=ScalarParameter() {defaultValue=1, group=WPO, parameterName=WPO Wiggle, sliderMax=10})
  n15 = Multiply(
   A=Multiply(A=ConstantBiasScale(None=@n12) {bias=0.8}, B=Constant3Vector() {constant=<rgba 0 0 1 0>})
   B=ScalarParameter() {defaultValue=1, group=WPO, parameterName=WPO Upwards, sliderMax=50}
  )
  n16 = DynamicParameter() {parameterIndex=1, paramNames [Position X, Position Y, Position Z, Param4]}
  n17 = Subtract(
   A=MaterialFunctionCall(X=@n16.Position X, Y=@n16.Position X, Z=@n16.Position X) {materialFunction=&MakeFloat3}.Result
   B=ParticlePositionWS()
  )
  n18 = ScalarParameter() {defaultValue=20, group=WPO, parameterName=WPO Max Drag, sliderMax=50}
  n19 = Multiply(A=@n18) {constB=-1}
  n20 = Add(
   A=Multiply(A=Add(A=@n14, B=@n15), B=ScalarParameter() {defaultValue=1, group=WPO, parameterName=WPO Wave, sliderMax=10, sortPriority=33})
   B=Multiply(
    A=VectorParameter() {group=WPO, parameterName=WPO Drag Direction Mask, defaultValue=<rgba 1 1 1 1>}.RGB
    B=Clamp(None=@n17, Min=@n19, Max=@n18)
   )
  )
  n21 = TextureCoordinate()
  n22 = Power(
   Base=OneMinus(None=ComponentMask(None=@n21) {g=true})
   Exp=ScalarParameter() {defaultValue=1, group=WPO, parameterName=WPO Bend, sliderMax=10}
  )
  Add(
   A=MaterialFunctionCall(
    WidthBase=Multiply(A=@n1, B=@n2.WidthBaseMultiplier)
    WidthTip=Multiply(A=@n2.WidthTipMultiplier, B=@n3)
    Expand U or V UV Channel=StaticBool() {value=true}
   ) {materialFunction=&SplineThicken}.WorldPositionOffset
   B=Multiply(A=ScalarParameter() {group=WPO, parameterName=WPO Intensity, sliderMax=100}, B=LinearInterpolate(B=@n20, Alpha=@n22))
  )
  n23 = VectorParameter() {group=Color, parameterName=Tip Color, defaultValue=<rgba 0 0 0 1>}
  n24 = Constant4Vector() {constant=<rgba 1 1 1 0>}
  n25 = MaterialFunctionCall(
   Color_Morning=@n24.RGBA
   Color_Day=@n24.RGBA
   Color_Evening=@n24.RGBA
   Color_Night=VectorParameter() {parameterName=Color_Night_Tip, defaultValue=<rgba 1 1 1 1>}.RGB
  ) {materialFunction=&Color_Time_of_Day}
  n26 = ParticleColor()
  n27 = TextureCoordinate()
  n28 = AppendVector(
   A=ComponentMask(None=@n27) {r=true}
   B=Power(Base=ComponentMask(None=@n27) {g=true}, Exp=ScalarParameter() {defaultValue=2, group=Wispy Flame, parameterName=UV Stretch, sliderMax=5})
  )
  n29 = AppendVector(
   A=ScalarParameter() {group=Wispy Flame, parameterName=Speed X, sliderMax=5, sliderMin=-5}
   B=ScalarParameter() {defaultValue=1, group=Wispy Flame, parameterName=Speed Y, sliderMax=5, sliderMin=-5}
  )
  n30 = Constant()
  n31 = DynamicParameter() {parameterIndex=2, defaultValue=<rgba 1 0 1 1>, paramNames [Speed Y offset, Swipe, Mesh Index, Param4]}
  n32 = Multiply(A=@n31.Speed Y offset, B=VertexColor().G)
  n33 = Multiply(
   A=Add(A=Panner(Coordinate=@n28, Speed=@n29) {speedX=1}, B=AppendVector(A=@n30, B=@n32))
   B=AppendVector(
    A=ScalarParameter() {defaultValue=1, group=Wispy Flame, parameterName=Scale X, sliderMax=5}
    B=ScalarParameter() {defaultValue=1, group=Wispy Flame, parameterName=Scale Y, sliderMax=5}
   )
  )
  n34 = AppendVector(A=VertexColor().R, B=Constant())
  n35 = TextureSampleParameter2D(UVs=MaterialFunctionCall(XY=Add(A=@n33, B=@n34)) {materialFunction=&Swizzle}.YX) {
   group=Wispy Flame
   parameterName=Flame Texture
   texture=/BRCosmetics/Athena/Effects/Particles/Skydiving/NS_Contrail_Chainmail/T_WispyFireNoise03
  }
  n36 = OneMinus(None=DynamicParameter() {paramNames [WidthTipMultiplier, WidthBaseMultiplier, Concave, U Length]}.U Length)
  n37 = Subtract(A=ComponentMask(None=TextureCoordinate()) {g=true}, B=@n36) {constB=0.1}
  n38 = ComponentMask(None=TextureCoordinate()) {g=true}
  n39 = DynamicParameter() {parameterIndex=2, defaultValue=<rgba 1 0 1 1>, paramNames [Speed Y offset, Swipe, Mesh Index, Param4]}
  n40 = Saturate(None=OneMinus(None=Add(A=@n38, B=@n39.Swipe)))
  n41 = Multiply(
   A=Divide(A=Saturate(None=@n37), B=OneMinus(None=@n36))
   B=StaticSwitchParameter(True=Sine(None=@n40) {period=2}, False=Constant() {r=1}) {parameterName=Swipe}
  )
  n42 = SphereMask(A=ComponentMask(None=TextureCoordinate()) {r=true}, B=Constant() {r=0.5}) {attenuationRadius=0.5, hardnessPercent=0}
  n43 = ScalarParameter() {defaultValue=1, group=Wispy Flame, parameterName=Concave, sliderMax=10}
  n44 = DynamicParameter() {paramNames [WidthTipMultiplier, WidthBaseMultiplier, Concave, U Length]}
  n45 = SmoothStep(Value=Power(Base=Multiply(A=@n41, B=@n42), Exp=Multiply(A=@n43, B=@n44.Concave))) {constMax=0.75}
  n46 = LinearInterpolate(A=Multiply(A=@n35.R, B=Multiply(A=@n45) {constB=2}), B=@n45, Alpha=@n45)
  n47 = SmoothStep(
   Min=ScalarParameter() {group=Color, parameterName=Tip Color Gradient Min, sliderMax=1}
   Max=ScalarParameter() {group=Color, parameterName=Tip Color Gradient Max, sliderMax=1}
   Value=@n46
  )
  n48 = SmoothStep(
   Min=ScalarParameter() {group=Color, parameterName=Tip Color UV Min, sliderMax=1}
   Max=ScalarParameter() {group=Color, parameterName=Tip Color UV Max, sliderMax=1}
   Value=ComponentMask(None=TextureCoordinate()) {g=true}
  )
  MaterialFunctionCall(Emissive=LinearInterpolate(A=Multiply(A=@n23.RGB, B=@n25.Result), B=@n26.RGB, Alpha=LinearInterpolate(A=@n47, Alpha=@n48)) {constB=0}) {materialFunction=&MF_FX_EmissiveAdjustments}
  n49 = SmoothStep(
   Max=OneMinus(None=ScalarParameter() {defaultValue=0.9, group=Wispy Flame, parameterName=Edge Hardness, sliderMax=1})
   Value=Multiply(A=@n26.A, B=@n46)
  ) {constMax=0.1}
  n50 = DepthFade(Opacity=@n49, FadeDistance=ScalarParameter() {defaultValue=35, group=Opacity, parameterName=Depth Fade Distance, sliderMax=100}) {fadeDistanceDefault=25}
  n51 = ShadingPathSwitch(Default=@n50, Deferred=@n50, Forward=@n49, Mobile=@n49)
  n52 = ScalarParameter() {defaultValue=1, group=Opacity, parameterName=Opacity, sliderMax=1}
  n53 = ScalarParameter() {group=Opacity, parameterName=Edge UV Fade, sliderMax=10}
  n54 = OneMinus(None=ComponentMask(None=TextureCoordinate()) {g=true})
  n55 = VertexColor()
  n56 = DynamicParameter() {parameterIndex=2, defaultValue=<rgba 1 0 1 1>, paramNames [Speed Y offset, Swipe, Mesh Index, Param4]}
  n57 = Constant()
  n58 = Constant() {r=1}
  Multiply(
   A=Multiply(A=Multiply(A=@n51, B=@n52), B=SmoothStep(Max=@n53, Value=@n54))
   B=StaticSwitchParameter(True=If(A=@n55.G, B=@n56.Mesh Index, "A > B"=@n57, "A == B"=@n58, "A < B"=@n57) {equalsThreshold=0.01}, False=Constant() {r=1}) {parameterName=Mesh Index}
  )
 ]
}
