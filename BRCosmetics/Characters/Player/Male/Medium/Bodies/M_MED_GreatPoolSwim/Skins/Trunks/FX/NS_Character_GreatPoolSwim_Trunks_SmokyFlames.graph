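# Niagara cosmetic effect for the GreatPoolSwim Trunks character skin, attaching smoky red flames, embers and glows to skeletal mesh bones.
# Burst-spawned flame meshes track bones, feeding smoothed positions to materials, while glow sprites and lights flicker by waveform.
# A system Timeline drives ember and smoke-sprite spawning, flame darkening and glow surges.
# User bools hide parts for ADS, hidden head or feet, and back/eye FX toggles.
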NiagaraSystem /BRCosmetics/Characters/Player/Male/Medium/Bodies/M_MED_GreatPoolSwim/Skins/Trunks/FX/NS_Character_GreatPoolSwim_Trunks_SmokyFlames
dependencies {
 usedDataInterfaces [
  &NiagaraDataInterfaceArrayFloat
  &NiagaraDataInterfaceArrayFloat3
  &NiagaraDataInterfaceCurve
  &NiagaraDataInterfaceArrayInt32
  &NiagaraDataInterfaceColorCurve
  &NiagaraDataInterfaceVector2DCurve
  &NiagaraDataInterfaceSkeletalMesh
 ]
 usedDynamicInputs [&SimulationPosition, &FNTimeOfDayScalar]
 usedModules [
  &self:SystemSpawnScript.SystemScriptSource.SystemScriptGraph.NiagaraNodeAssignment_9.SetVariables_#1
  &self:SystemSpawnScript.SystemScriptSource.SystemScriptGraph.NiagaraNodeAssignment_0.SetVariables_#2
  &SystemState
  &self:SystemSpawnScript.SystemScriptSource.SystemScriptGraph.NiagaraNodeAssignment_5.SetVariables_#3
  &self:SystemSpawnScript.SystemScriptSource.SystemScriptGraph.NiagaraNodeAssignment_1.SetVariables_#4
  &Timeline
  &self:SystemSpawnScript.SystemScriptSource.SystemScriptGraph.NiagaraNodeAssignment_2.SetVariables_#5
 ]
}
emitters {
 Embers {
  emitterUpdateScript [
   &EmitterState(Life Cycle Mode=::NewEnumerator1, Loop Behavior=::NewEnumerator1, Loop Duration Mode=::NewEnumerator1, UseLoopDelay=false)
   &SpawnRate(
    SpawnRate=&FloatFromCurve(
     FloatCurve=DataInterface() {propertyValues {Curve {keys [
         {arriveTangent=-0.6696, interpMode=RCIM_Cubic, leaveTangent=-0.6696, tangentMode=RCTM_User, time=0, value=1}
         {interpMode=RCIM_Cubic, tangentMode=RCTM_User, time=1, value=0}
        ]}}}
     CurveIndex=$System.Timeline.Time
     Scale Curve=600
    )
    Use Spawn Probability=true
   ) {module=SpawnRate001}
  ]
  particleSpawnScript [
   /SetVariables(Particles.PerParticleRandom:NiagaraFloat=&RandomRangeFloat(Minimum=0, Maximum=1)) {isSetParameters=true}
   &FN_SetSpawnGroup()
   &InitializeParticle(
    Color=<rgba 1 0 0 1>
    Lifetime=&RandomRangeFloat(Minimum=0.2, Maximum=0.6)
    Sprite Size=&Multiply_Vector2D(
     A=&RandomRangeVector2D(Minimum=<0.15 0.2>, Maximum=<0.2 0.4>)
     B=&MakeVector2D(
      X=&FloatFromCurve(
       FloatCurve=DataInterface() {propertyValues {Curve {keys [<at 0 1>,<at 0.8 2 RCIM_Constant>,<at 0.95 3 RCIM_Constant>]}}}
       CurveIndex=$Particles.PerParticleRandom
      )
      Y=&FloatFromCurve(
       FloatCurve=DataInterface() {propertyValues {Curve {keys [<at 0 1>,<at 0.8 2 RCIM_Constant>,<at 0.95 3>]}}}
       CurveIndex=$Particles.PerParticleRandom
      )
     )
    )
    Sprite Size Mode=::NewEnumerator4
    UsePositionOffset=false
    Write Material Random=true
   )
   &SkeletalMeshLocation(
    Bone / Socket Index=&SelectIntFromArray(
     Direct Array Index=/Niagara/DynamicInputs/Divide/Divide_Int(A:NiagaraInt32=&SpawnGroup(), B:NiagaraInt32=2)
     Int Selection Array=DataInterface() {propertyValues {IntData [0,5]}}
    )
    Bone Sampling Mode=::NewEnumerator4
    "Inherit Velocity (Scale)"=0.2
    MaskSpawnGroup=false
    Orientation Sampling=::NewEnumerator1
    Sampled Position Offset=&SelectVectorFromArray(
     Direct Array Index=&SpawnGroup()
     Vector Selection Array=DataInterface() {propertyValues {FloatData [<7.5 13 9>, <7.5 13 -9>, <5 -12 3>, <5 -12 -3>]}}
    )
    Skeletal Mesh=$System.Skeletal Mesh
    Velocity Sampling=::NewEnumerator0
    VelocitySpeedLimit=false
   )
   &ShapeLocation(MaskSpawnGroup=false, Override Seed=false, Sphere Radius=1, UseOwnerScale=true)
   &AddVelocity(
    Cone Angle=90
    Cone Axis=&RotateVectorByQuaternion(
     Vector To Rotate=&SelectVectorFromArray(
      Direct Array Index=&SpawnGroup()
      Vector Selection Array=DataInterface() {propertyValues {FloatData [<1 0.3 1>, <1 0.3 -1>, <1 -0.2 0.2>, <1 -0.2 -0.2>]}}
     )
     Quaternion=$Particles.SkeletalMeshLocation.SampledRotation
    )
    Module.UseConeFalloff=false
    Module.UseVelDistribution=false
    Override Seed=false
    Rotation Coordinate Space=::Simulation
    Velocity Mode=::NewEnumerator2
    Velocity Speed=&RandomRangeFloat(Minimum=20, Maximum=40)
   )
   &KillParticles(Kill Particles=&Boolean_AndOperation(
     Bool1=&Boolean_NotOperation(Bool=$System.bShowBackFX)
     Bool2=&SetBoolByIntComparison(A=&SpawnGroup(), B=2, Comparison Type=::NewEnumerator2)
    ))
  ]
  particleUpdateScript [
   &ParticleState()
   &Drag(
    Drag=&FloatFromCurve(
     FloatCurve=DataInterface() {propertyValues {Curve {keys [
         {interpMode=RCIM_Cubic, tangentMode=RCTM_User, time=0, value=0}
         {arriveTangent=2.3593, interpMode=RCIM_Cubic, leaveTangent=2.3593, tangentMode=RCTM_User, time=1, value=1}
        ]}}}
     Scale Curve=&RandomRangeFloat(Minimum=15, Maximum=25)
    )
    Use Linear Drag=true
    Use Rotational Drag=false
   )
   &CurlNoiseForce(
    Noise Frequency=75
    Noise Strength=&FloatFromCurve(
     FloatCurve=DataInterface() {propertyValues {Curve {keys [{interpMode=RCIM_Cubic, tangentMode=RCTM_User, time=0, value=0}, {interpMode=RCIM_Cubic, tangentMode=RCTM_User, time=1, value=1}]}}}
     Scale Curve=500
    )
    Pan Noise=true
    Pan Noise Field=<0 0 -0.5>
   ) {module=CurlNoiseForce001}
   &GravityForce(Gravity=&MakeVector(
     X=0
     Y=0
     Z=&FloatFromCurve(
      FloatCurve=DataInterface() {propertyValues {Curve {keys [{interpMode=RCIM_Cubic, tangentMode=RCTM_User, time=0, value=0}, {interpMode=RCIM_Cubic, tangentMode=RCTM_User, time=1, value=1}]}}}
      Scale Curve=50
     )
    ))
   &SolveForcesAndVelocity(Clamp Velocity=false, Limit Acceleration=false)
   &ScaleColor(
    Scale RGB=&MakeVectorFromLinearColorRGB(Color=&ColorFromCurve(ColorCurve=DataInterface() {propertyValues {
        AlphaCurve {keys [<at 0.571 1 RCIM_Cubic>,<at 0.9966 0>]}
        BlueCurve {keys [<at 0.0039 3>,<at 1 1 RCIM_Cubic>]}
        GreenCurve {keys [<at 0.0039 3>,<at 1 1 RCIM_Cubic>]}
        RedCurve {keys [<at 0.0039 3>,<at 1 1 RCIM_Cubic>]}
       }}))
    ScaleA=true
    ScaleRGB=true
   )
   &ScaleSpriteSize(
    Uniform Curve Scale=&MakeFloatFromVector(Vector=$Engine.Owner.Scale, Channel=::NewEnumerator0)
    Uniform Curve Sprite Scale=DataInterface() {propertyValues {Curve {keys [
        {tangentMode=RCTM_User, time=0, value=1}
        {arriveTangent=-0.0405, interpMode=RCIM_Cubic, leaveTangent=-0.0405, tangentMode=RCTM_User, time=0.7009, value=1}
        {arriveTangent=-8.0089, interpMode=RCIM_Cubic, leaveTangent=-8.0089, tangentMode=RCTM_User, time=1, value=0}
       ]}}}
   )
   &ScaleSpriteSizeBySpeed(
    Max Scale Factor=<1.666 10>
    Min Scale Factor=<1.333 1.333>
    Sample Scale Factor By Curve=true
    Scale Factor Curve=DataInterface() {propertyValues {Curve {keys [<at 0 0>, {arriveTangent=1.4989, interpMode=RCIM_Cubic, tangentMode=RCTM_Break, time=1, value=1}]}}}
    Velocity Threshold=150
   )
   &ScaleSpriteSize(
    Scale Sprite Size Mode=::NewEnumerator0
    Uniform Scale Factor=&MakeCustomFloatFromBool(
     Bool=$User.bisADS
     True Float=&SelectFloatFromArray(Direct Array Index=$Particles.SpawnGroup, Float Selection Array=DataInterface() {propertyValues {FloatData [0,0,1,1]}})
     False Float=1
    )
   ) {module=ScaleSpriteSize001}
  ]
  rendererClasses [&NiagaraSpriteRendererProperties]
  renderers [{rendererClass=&NiagaraSpriteRendererProperties}]
 }
 FlameSprites {
  emitterUpdateScript [
   &EmitterState(Life Cycle Mode=::NewEnumerator1, Loop Behavior=::NewEnumerator1, Loop Delay=1.1, Loop Duration Mode=::NewEnumerator1, UseLoopDelay=false)
   &SpawnRate(
    SpawnRate=&FloatFromCurve(
     FloatCurve=DataInterface() {propertyValues {Curve {keys [
         {arriveTangent=-0.5725, interpMode=RCIM_Cubic, leaveTangent=-0.5725, tangentMode=RCTM_User, time=0, value=1}
         {interpMode=RCIM_Cubic, tangentMode=RCTM_User, time=1, value=0}
        ]}}}
     CurveIndex=$System.Timeline.Time
     Scale Curve=150
    )
    Use Spawn Probability=true
   ) {module=SpawnRate001}
  ]
  particleSpawnScript [
   &InitializeParticle(
    Color Maximum=<rgba 0.03 0 0 0.6>
    Color Minimum=<rgba 0.5 0.025 0 0.4>
    Color Mode=::NewEnumerator2
    Lifetime=&Multiply_Float(
     A=&RandomRangeFloat(Minimum=1, Maximum=1.2)
     B=&FloatFromCurve(
      FloatCurve=DataInterface() {propertyValues {Curve {keys [
          {arriveTangent=1.9127, interpMode=RCIM_Cubic, leaveTangent=1.9127, tangentMode=RCTM_User, time=0, value=1}
          <at 0.2 1.2 RCIM_Cubic>
          <at 1 1 RCIM_Cubic>
         ]}}}
      CurveIndex=$System.Timeline.Time
     )
    )
    Sprite Rotation Mode=::NewEnumerator2
    Sprite Size Mode=::NewEnumerator1
    Sprite UV Mode=::NewEnumerator3
    Uniform Sprite Size Max=12.5
    Uniform Sprite Size Min=7.5
    UsePositionOffset=false
    Write Material Random=true
   ) {module=InitializeParticle001}
   &FN_SetSpawnGroup()
   &SkeletalMeshLocation(
    Bone / Socket Index=&SelectIntFromArray(
     Direct Array Index=&Divide_Int(A:NiagaraInt32=&SpawnGroup(), B:NiagaraInt32=2)
     Int Selection Array=DataInterface() {propertyValues {IntData [0,5]}}
    )
    Bone Sampling Mode=::NewEnumerator4
    MaskSpawnGroup=false
    Orientation Sampling=::NewEnumerator1
    Sampled Position Offset=&SelectVectorFromArray(
     Direct Array Index=&SpawnGroup()
     Vector Selection Array=DataInterface() {propertyValues {FloatData [<7.5 13 9>, <7.5 13 -9>, <5 -12 3>, <5 -12 -3>]}}
    )
    Skeletal Mesh=$System.Skeletal Mesh
   )
   &ShapeLocation(MaskSpawnGroup=false, Override Seed=false, Sphere Radius=3, UseOwnerScale=true)
   &CameraOffset(Camera Offset Amount=&RandomRangeFloat(Minimum=0, Maximum=10))
   &AddVelocity(
    Cone Angle=10
    Cone Axis=&RotateVectorByQuaternion(
     Vector To Rotate=&SelectVectorFromArray(
      Direct Array Index=&SpawnGroup()
      Vector Selection Array=DataInterface() {propertyValues {FloatData [<1 0.3 1>, <1 0.3 -1>, <1 -0.2 0.2>, <1 -0.2 -0.2>]}}
     )
     Quaternion=$Particles.SkeletalMeshLocation.SampledRotation
    )
    Module.UseConeFalloff=true
    Module.UseVelDistribution=false
    Override Seed=false
    Rotation Coordinate Space=::Simulation
    Velocity Mode=::NewEnumerator2
    Velocity Speed=&RandomRangeFloat(Minimum=10, Maximum=50)
   )
   &KillParticles(Kill Particles=&Boolean_AndOperation(
     Bool1=&Boolean_NotOperation(Bool=$System.bShowBackFX)
     Bool2=&SetBoolByIntComparison(A=&SpawnGroup(), B=2, Comparison Type=::NewEnumerator2)
    ))
  ]
  particleUpdateScript [
   /SetVariables(Particles.SubImageIndex=&FloatFromCurve(FloatCurve=DataInterface() {propertyValues {Curve {keys [<at 0 0>,<at 1 15>]}}})) {isSetParameters=true}
   &ParticleState()
   &ScaleSpriteSize(Scale Factor=&Vector2DFromCurve(
     Vector2Curve=DataInterface() {propertyValues {
       XCurve {keys [
         {arriveTangent=2.2044, interpMode=RCIM_Cubic, leaveTangent=2.2044, tangentMode=RCTM_User, time=0, value=0.4}
         {interpMode=RCIM_Cubic, tangentMode=RCTM_User, time=0.5, value=1}
         {arriveTangent=-1.7158, tangentMode=RCTM_User, time=1, value=0.05}
        ]}
       YCurve {keys [
         {arriveTangent=2.2044, interpMode=RCIM_Cubic, leaveTangent=2.2044, tangentMode=RCTM_User, time=0, value=0.4}
         {interpMode=RCIM_Cubic, tangentMode=RCTM_User, time=0.5, value=1}
         {arriveTangent=-1.7158, tangentMode=RCTM_User, time=1, value=0.05}
        ]}
      }}
     ScaleCurve=<1 1>
    ))
   &ScaleColor(
    Scale Alpha=&Multiply_Float(
     A=&FloatFromCurve(
      FloatCurve=DataInterface() {propertyValues {Curve {keys [
          {arriveTangent=9.0725, interpMode=RCIM_Cubic, leaveTangent=9.0725, tangentMode=RCTM_User, time=0, value=0}
          {arriveTangent=0.0164, interpMode=RCIM_Cubic, leaveTangent=0.0164, tangentMode=RCTM_User, time=0.1408, value=0.9847}
          {arriveTangent=-2.2898, interpMode=RCIM_Cubic, leaveTangent=-2.2898, tangentMode=RCTM_User, time=1, value=0}
         ]}}}
      Scale Curve=&FN_FrontEndFloatModifier(Gameplay=0.35, Front End=0.25)
     )
     B=&RandomRangeFloat(Minimum=1, Maximum=2)
    )
    Scale RGB=&VectorFromFloat(Value=&FloatFromCurve(FloatCurve=DataInterface() {propertyValues {Curve {keys [
          {arriveTangent=-3.4737, interpMode=RCIM_Cubic, leaveTangent=-3.4737, tangentMode=RCTM_User, time=0, value=1}
          {arriveTangent=-0.0056, interpMode=RCIM_Cubic, leaveTangent=-0.0056, tangentMode=RCTM_User, time=0.5, value=0}
         ]}}}))
    ScaleA=true
    ScaleRGB=true
   ) {module=ScaleColor001}
   &Drag(
    Drag=&FloatFromCurve(
     FloatCurve=DataInterface() {propertyValues {Curve {keys [
         {arriveTangent=4.9807, interpMode=RCIM_Cubic, leaveTangent=4.9807, tangentMode=RCTM_User, time=0, value=0}
         {interpMode=RCIM_Cubic, tangentMode=RCTM_User, time=1, value=2}
        ]}}}
     Scale Curve=5
    )
    Use Linear Drag=true
    Use Rotational Drag=true
   )
   &CurlNoiseForce(
    Noise Frequency=100
    Noise Strength=&FloatFromCurve(
     FloatCurve=DataInterface() {propertyValues {Curve {keys [{interpMode=RCIM_Cubic, tangentMode=RCTM_User, time=0, value=0}, {interpMode=RCIM_Cubic, tangentMode=RCTM_User, time=1, value=1}]}}}
     Scale Curve=300
    )
    Pan Noise=true
    Pan Noise Field=<0 0 0.1>
   )
   &GravityForce(Gravity=&MakeVector(
     X=0
     Y=0
     Z=&FloatFromCurve(
      FloatCurve=DataInterface() {propertyValues {Curve {keys [
          {arriveTangent=2.1832, interpMode=RCIM_Cubic, leaveTangent=2.1832, tangentMode=RCTM_User, time=0, value=0}
          {interpMode=RCIM_Cubic, tangentMode=RCTM_User, time=1, value=1}
         ]}}}
      Scale Curve=50
     )
    ))
   &InheritSourceMovement(
    Applied Position Scale=&VectorFromFloat(Value=&FloatFromCurve(FloatCurve=DataInterface() {propertyValues {Curve {keys [
          {interpMode=RCIM_Cubic, tangentMode=RCTM_User, time=0, value=1}
          {arriveTangent=-2.5265, interpMode=RCIM_Cubic, leaveTangent=-2.5265, tangentMode=RCTM_User, time=1, value=0}
         ]}}}))
    ApplyForce=false
    ApplyPosition=true
    ApplyPositionFalloff=false
   )
   &SolveForcesAndVelocity(Clamp Velocity=false, Limit Acceleration=false)
   &ScaleSpriteSize(
    Scale Sprite Size Mode=::NewEnumerator0
    Uniform Scale Factor=&MakeCustomFloatFromBool(
     Bool=$User.bisADS
     True Float=&SelectFloatFromArray(Direct Array Index=$Particles.SpawnGroup, Float Selection Array=DataInterface() {propertyValues {FloatData [0,0,1,1]}})
     False Float=1
    )
   ) {module=ScaleSpriteSize001}
  ]
  rendererClasses [&NiagaraSpriteRendererProperties]
  renderers [{rendererClass=&NiagaraSpriteRendererProperties}]
 }
 Flames_Spline_Mesh {
  emitterUpdateScript [
   &EmitterState(
    Inactive Response=::NewEnumerator1
    Life Cycle Mode=::NewEnumerator1
    Loop Behavior=::NewEnumerator1
    Loop Duration Mode=::NewEnumerator1
    UseLoopDelay=false
   )
   &SpawnBurst_Instantaneous(
    Spawn Count=&MakeCustomIntFromBool(Bool=$System.bShowEyeFX, True INT=8, False INT=6)
    Use Spawn Probability=false
    UseLoopCountLimit=false
   )
  ]
  particleSpawnScript [
   &FN_SetSpawnGroup(Number of Spawn Groups=8)
   /SetVariables(Particles.VisibilityTag=&SpawnGroup()) {isSetParameters=true}
   &InitializeParticle(
    Color=$System.TOD Flame Color
    Lifetime=&RandomRangeFloat(Minimum=0.45, Maximum=0.6)
    Mesh Scale Mode=::NewEnumerator3
    UsePositionOffset=false
    Write Material Random=true
   )
   &CameraOffset(Camera Offset Amount=&SelectFloatFromArray(
     Direct Array Index=&SpawnGroup()
     Float Selection Array=DataInterface() {propertyValues {FloatData [1,2,1,0.5,5,0.5,1]}}
    ))
   /SetVariables() {isSetParameters=true}
   &KillParticles(Kill Particles=&Boolean_AndOperation(
     Bool1=&Boolean_NotOperation(Bool=$System.bShowBackFX)
     Bool2=&SetBoolByIntComparison(A=&SpawnGroup(), B=5, Comparison Type=::NewEnumerator4)
    ))
  ]
  particleUpdateScript [
   &ParticleState(Kill Particles When Lifetime Has Elapsed=false)
   &SkeletalMeshLocation(
    Bone / Socket Index=&SpawnGroup()
    Bone Sampling Mode=::NewEnumerator4
    MaskSpawnGroup=false
    Sampled Position Offset=&SelectVectorFromArray(
     Direct Array Index=&SpawnGroup()
     Vector Selection Array=DataInterface() {propertyValues {FloatData [<1.5 0 0>, <0 0 0>, <0 0 0>, <-0.1 -0.2 0.1>, <0 0 0>, <-1 0.5 0>, <0 -0.1 0>]}}
    )
    Skeletal Mesh=$System.Skeletal Mesh
   ) {module=SkeletalMeshLocation001}
   &InterpolateOverTime(Rate Of Change=7.5, Target Vector Value:Vector3f=$Particles.SkeletalMeshLocation001.SampledPosition, Variable Type=::NewEnumerator2)
   /SetVariables(Particles.Y Offset:NiagaraFloat=&Add_Float(
     A=$Particles.Y Offset
     B=&Lerp_Float(
      A=0
      B=0.02
      Alpha=&FloatFromCurve(
       FloatCurve=DataInterface() {propertyValues {Curve {keys [
           {arriveTangent=2.1926, interpMode=RCIM_Cubic, leaveTangent=2.1926, tangentMode=RCTM_User, time=0, value=0}
           {arriveTangent=0.0606, interpMode=RCIM_Cubic, leaveTangent=0.0606, tangentMode=RCTM_User, time=0.3, value=1}
           {arriveTangent=-0.026, interpMode=RCIM_Cubic, leaveTangent=-0.026, tangentMode=RCTM_User, time=0.7, value=1}
           {arriveTangent=-2.0344, interpMode=RCIM_Cubic, leaveTangent=-2.0344, tangentMode=RCTM_User, time=1, value=0}
          ]}}}
       CurveIndex=$System.Timeline.Time
      )
     )
    )) {isSetParameters=true}
   &DynamicMaterialParameters(
    Index 0 Param 1=&Lerp_Float(
     A=1
     B=2
     Alpha=&FloatFromCurve(
      FloatCurve=DataInterface() {propertyValues {Curve {keys [
          {arriveTangent=10.3645, interpMode=RCIM_Cubic, leaveTangent=10.3645, tangentMode=RCTM_User, time=0, value=0}
          {arriveTangent=-0.0325, interpMode=RCIM_Cubic, leaveTangent=-0.0325, tangentMode=RCTM_User, time=0.2, value=1}
          {arriveTangent=-0.026, interpMode=RCIM_Cubic, leaveTangent=-0.026, tangentMode=RCTM_User, time=0.8, value=1}
          {arriveTangent=-3.2787, interpMode=RCIM_Cubic, leaveTangent=-3.2787, tangentMode=RCTM_User, time=1, value=0}
         ]}}}
      CurveIndex=$System.Timeline.Time
     )
    )
    Index 0 Param 2=&Lerp_Float(
     A=1
     B=0.5
     Alpha=&FloatFromCurve(
      FloatCurve=DataInterface() {propertyValues {Curve {keys [
          {arriveTangent=10.3645, interpMode=RCIM_Cubic, leaveTangent=10.3645, tangentMode=RCTM_User, time=0, value=0}
          {arriveTangent=-0.0325, interpMode=RCIM_Cubic, leaveTangent=-0.0325, tangentMode=RCTM_User, time=0.2, value=1}
          {arriveTangent=-0.026, interpMode=RCIM_Cubic, leaveTangent=-0.026, tangentMode=RCTM_User, time=0.8, value=1}
          {arriveTangent=-3.2787, interpMode=RCIM_Cubic, leaveTangent=-3.2787, tangentMode=RCTM_User, time=1, value=0}
         ]}}}
      CurveIndex=$System.Timeline.Time
     )
    )
    Index 0 Param 3=&Lerp_Float(
     A=1
     B=0.7
     Alpha=&FloatFromCurve(
      FloatCurve=DataInterface() {propertyValues {Curve {keys [
          {arriveTangent=10.3645, interpMode=RCIM_Cubic, leaveTangent=10.3645, tangentMode=RCTM_User, time=0, value=0}
          {arriveTangent=-0.0325, interpMode=RCIM_Cubic, leaveTangent=-0.0325, tangentMode=RCTM_User, time=0.2, value=1}
          {arriveTangent=-0.026, interpMode=RCIM_Cubic, leaveTangent=-0.026, tangentMode=RCTM_User, time=0.8, value=1}
          {arriveTangent=-3.2787, interpMode=RCIM_Cubic, leaveTangent=-3.2787, tangentMode=RCTM_User, time=1, value=0}
         ]}}}
      CurveIndex=$System.Timeline.Time
     )
    )
    Index 0 Param 4=&Lerp_Float(
     A=0.8
     B=1
     Alpha=&FloatFromCurve(
      FloatCurve=DataInterface() {propertyValues {Curve {keys [
          {arriveTangent=10.3645, interpMode=RCIM_Cubic, leaveTangent=10.3645, tangentMode=RCTM_User, time=0, value=0}
          {arriveTangent=-0.0325, interpMode=RCIM_Cubic, leaveTangent=-0.0325, tangentMode=RCTM_User, time=0.2, value=1}
          {arriveTangent=-0.026, interpMode=RCIM_Cubic, leaveTangent=-0.026, tangentMode=RCTM_User, time=0.8, value=1}
          {arriveTangent=-3.2787, interpMode=RCIM_Cubic, leaveTangent=-3.2787, tangentMode=RCTM_User, time=1, value=0}
         ]}}}
      CurveIndex=$System.Timeline.Time
     )
    )
    Index 1 Param 1=&MakeFloatFromVector(Vector=$StackContext.InterpolateOverTime.InterpolateOverTime_V3_.Moving Average, Channel=::NewEnumerator0)
    Index 1 Param 2=&MakeFloatFromVector(Vector=$StackContext.InterpolateOverTime.InterpolateOverTime_V3_.Moving Average, Channel=::NewEnumerator1)
    Index 1 Param 3=&MakeFloatFromVector(Vector=$StackContext.InterpolateOverTime.InterpolateOverTime_V3_.Moving Average, Channel=::NewEnumerator2)
    Index 2 Param 1=$Particles.Y Offset
    Module.Param0WriteEnabled=true
    Module.Param0WriteEnabled1=true
    Module.Param0WriteEnabled2=true
    Module.Param1WriteEnabled=true
    Module.Param1WriteEnabled1=true
    Module.Param1WriteEnabled2=false
    Module.Param2WriteEnabled=true
    Module.Param2WriteEnabled1=true
    Module.Param2WriteEnabled2=false
    Module.Param3WriteEnabled=true
    Module.Param3WriteEnabled1=true
    Module.Param3WriteEnabled2=false
    Write Parameter Index 1=true
    Write Parameter Index 2=true
   ) {module=DynamicMaterialParameters001}
   /SetVariables(Particles.Dynamic Parameter Dark:Vector4f=&MakeVector4(
     X=1
     Y=1
     Z=&Lerp_Float(
      A=1
      B=0.8
      Alpha=&FloatFromCurve(
       FloatCurve=DataInterface() {propertyValues {Curve {keys [
           {arriveTangent=10.3645, interpMode=RCIM_Cubic, leaveTangent=10.3645, tangentMode=RCTM_User, time=0, value=0}
           {arriveTangent=-0.0325, interpMode=RCIM_Cubic, leaveTangent=-0.0325, tangentMode=RCTM_User, time=0.2, value=1}
           {arriveTangent=-0.026, interpMode=RCIM_Cubic, leaveTangent=-0.026, tangentMode=RCTM_User, time=0.8, value=1}
           {arriveTangent=-3.2787, interpMode=RCIM_Cubic, leaveTangent=-3.2787, tangentMode=RCTM_User, time=1, value=0}
          ]}}}
       CurveIndex=$System.Timeline.Time
      )
     )
     W=&Lerp_Float(
      A=0.7
      B=1
      Alpha=&FloatFromCurve(
       FloatCurve=DataInterface() {propertyValues {Curve {keys [
           {arriveTangent=10.3645, interpMode=RCIM_Cubic, leaveTangent=10.3645, tangentMode=RCTM_User, time=0, value=0}
           {arriveTangent=-0.0325, interpMode=RCIM_Cubic, leaveTangent=-0.0325, tangentMode=RCTM_User, time=0.2, value=1}
           {arriveTangent=-0.026, interpMode=RCIM_Cubic, leaveTangent=-0.026, tangentMode=RCTM_User, time=0.8, value=1}
           {arriveTangent=-3.2787, interpMode=RCIM_Cubic, leaveTangent=-3.2787, tangentMode=RCTM_User, time=1, value=0}
          ]}}}
       CurveIndex=$System.Timeline.Time
      )
     )
    )) {isSetParameters=true}
   /SetVariables(Particles.Inner Color:LinearColor=&MultiplyLinearColor(
     Linear Color=<rgba 0.75 0 0 1>
     Scale Factor=&MakeVector4_FromFloat(Float=&FloatFromCurve(
       FloatCurve=DataInterface() {propertyValues {Curve {keys [
           {arriveTangent=6.4246, interpMode=RCIM_Cubic, leaveTangent=6.4246, tangentMode=RCTM_User, time=0, value=1}
           <at 0.2 1.65 RCIM_Cubic>
           {arriveTangent=-2.4497, interpMode=RCIM_Cubic, leaveTangent=-2.4497, tangentMode=RCTM_User, time=1, value=1}
          ]}}}
       CurveIndex=$System.Timeline.Time
      ))
    )) {isSetParameters=true}
   /SetVariables(Particles.Color=&LerpLinearColors(
     StartColor=$System.TOD Flame Color
     EndColor=<rgba 0.02 0.0061 0.0064 1>
     LerpFactor=&FloatFromCurve(
      FloatCurve=DataInterface() {propertyValues {Curve {keys [
          {arriveTangent=10.3645, interpMode=RCIM_Cubic, leaveTangent=10.3645, tangentMode=RCTM_User, time=0, value=0}
          <at 0.2 1 RCIM_Cubic>
          {arriveTangent=-1.3747, interpMode=RCIM_Cubic, leaveTangent=-1.3747, tangentMode=RCTM_User, time=1, value=0}
         ]}}}
      CurveIndex=$System.Timeline.Time
     )
    )) {isSetParameters=true}
   &ScaleMeshSize(Scale Factor=&VectorFromFloat(Value=&MakeCustomFloatFromBool(
      Bool=&Boolean_OrOperation(Bool1=$User.bisADS, Bool2=$User.bIsADSOther)
      True Float=&SelectFloatFromArray(Direct Array Index=$Particles.SpawnGroup, Float Selection Array=DataInterface() {propertyValues {FloatData [1,1,1,1,1,1,0,1]}})
      False Float=1
     )))
   &ScaleMeshSize(Scale Factor=&VectorFromFloat(Value=&MakeCustomFloatFromBool(
      Bool=$User.bisADS
      True Float=&SelectFloatFromArray(Direct Array Index=$Particles.SpawnGroup, Float Selection Array=DataInterface() {propertyValues {FloatData [0,0,1,1,1,1,1,1]}})
      False Float=1
     ))) {module=ScaleMeshSize001}
   /SetVariables(Particles.VisibilityTag=&MakeCustomIntFromBool(
     Bool=&SetBoolByIntComparison(A=&SpawnGroup(), B=6, Comparison Type=::NewEnumerator2)
     True INT=&MakeCustomIntFromBool(Bool=$User.bisHeadHidden, True INT=10, False INT=$Particles.VisibilityTag)
     False INT=$Particles.VisibilityTag
    )) {isSetParameters=true}
   /SetVariables(Particles.VisibilityTag=&MakeCustomIntFromBool(
     Bool=&SetBoolByIntComparison(A=&SpawnGroup(), B=2, Comparison Type=::NewEnumerator4)
     True INT=&MakeCustomIntFromBool(Bool=$User.bisFootHidden, True INT=10, False INT=$Particles.VisibilityTag)
     False INT=$Particles.VisibilityTag
    )) {isSetParameters=true}
  ]
  rendererClasses [&NiagaraMeshRendererProperties]
  renderers [
   {rendererClass=&NiagaraMeshRendererProperties, rendererIndex=0}
   {rendererClass=&NiagaraMeshRendererProperties, rendererIndex=1}
   {rendererClass=&NiagaraMeshRendererProperties, rendererIndex=2}
   {rendererClass=&NiagaraMeshRendererProperties, rendererIndex=3}
   {rendererClass=&NiagaraMeshRendererProperties, rendererIndex=4}
   {rendererClass=&NiagaraMeshRendererProperties, rendererIndex=5}
   {rendererClass=&NiagaraMeshRendererProperties, rendererIndex=6}
   {rendererClass=&NiagaraMeshRendererProperties, rendererIndex=7}
   {rendererClass=&NiagaraMeshRendererProperties, rendererIndex=8}
   {rendererClass=&NiagaraMeshRendererProperties, rendererIndex=9}
   {rendererClass=&NiagaraMeshRendererProperties, rendererIndex=10}
   {rendererClass=&NiagaraMeshRendererProperties, rendererIndex=11}
   {rendererClass=&NiagaraMeshRendererProperties, rendererIndex=12}
   {rendererClass=&NiagaraMeshRendererProperties, rendererIndex=13}
  ]
 }
 GlowsAndLights {
  emitterUpdateScript [
   &EmitterState(
    Inactive Response=::NewEnumerator1
    Life Cycle Mode=::NewEnumerator1
    Loop Behavior=::NewEnumerator1
    Loop Duration Mode=::NewEnumerator1
    UseLoopDelay=false
   )
   &SpawnBurst_Instantaneous(Spawn Count=10, Use Spawn Probability=false, UseLoopCountLimit=true)
  ]
  particleSpawnScript [
   &PartitionParticles(ID to Partition:NiagaraInt32=&ReturnExecIndex(), Number Of Partitions=10)
   &InitializeParticle(Color Mode=::NewEnumerator0, UsePositionOffset=false, Write Material Random=true)
   /SetVariables(
    Particles.GlowsVisibilityTag:NiagaraInt32=&MergeHelper_IntSwitch(
     0=0
     1=0
     2=0
     3=0
     4=0
     5=&MakeCustomIntFromBool(Bool=$System.bShowBackFX, True INT=0, False INT=1)
     6=&MakeCustomIntFromBool(Bool=$System.bShowBackFX, True INT=0, False INT=1)
     7=&MakeCustomIntFromBool(Bool=$System.bShowEyeFX, True INT=0, False INT=1)
     8=&MakeCustomIntFromBool(Bool=$System.bShowEyeFX, True INT=0, False INT=1)
     9=0
     Index=$Particles.PartitionParticles.Partition
     Num=::NewEnumerator8
    )
    Particles.VisibilityTag=&SelectIntFromArray(
     Direct Array Index=$Particles.PartitionParticles.Partition
     Int Selection Array=DataInterface() {propertyValues {IntData [0,0,1,1,1,1,1,1,1,1]}}
    )
   ) {isSetParameters=true}
  ]
  particleUpdateScript [
   &ParticleState(Kill Particles When Lifetime Has Elapsed=false, Let Infinitely Lived Particles Die When Emitter Deactivates=true)
   /SetVariables(Particles.MainColor:LinearColor=&MultiplyLinearColor(
     Linear Color=<rgba 0.2 0 0 0.75>
     Scale Factor=&MakeVector4_FromFloat(Float=&Multiply_Float(
       A=&FloatFromCurve(
        FloatCurve=DataInterface() {propertyValues {Curve {keys [
            <at 0 1 RCIM_Cubic>
            <at 0.2 2 RCIM_Cubic>
            {arriveTangent=-0.0903, interpMode=RCIM_Cubic, leaveTangent=-0.0903, tangentMode=RCTM_User, time=1, value=1}
           ]}}}
        CurveIndex=$System.Timeline.Time
       )
       B=&Waveform(
        Waveform Count=::NewEnumerator1
        Global Input Position=&ReturnNormalizedExecIndex(Normalized Index Scale=9)
        "Waveform [1]"=::NewEnumerator12
        "[1] Frequency"=2.65
        "[1] Amplitude Min/Max"=<0.7 1>
        "Waveform [2] Blend"=::NewEnumerator5
        "[2] Interpolation"=&FloatFromCurve(
         FloatCurve=DataInterface() {propertyValues {Curve {keys [
             {tangentMode=RCTM_User, time=0, value=0}
             <at 0.1 1>
             {arriveTangent=-0.0405, leaveTangent=-0.0405, time=0.9, value=1}
             {arriveTangent=-8.0089, interpMode=RCIM_Cubic, leaveTangent=-8.0089, tangentMode=RCTM_User, time=1, value=0}
            ]}}}
         CurveIndex=$System.Timeline.Time
        )
        "[2] Invert Interpolation"=false
        "Waveform [2]"=::NewEnumerator12
        "[2] Input Position"=0
        "[2] Phase"=$Emitter.Age
        "[2] Phase Scale"=1
        "[2] Compound Frequency"=2
        "[2] Frequency"=16
        "[2] Amplitude Min/Max"=<0.75 1>
        "[2] Scale"=1
        "[2] Bias"=0
        "[2] Exponent"=1
        "[2] Pulse Duration"=0.1
        "[2] Clip Triangle Min/Max"=<0 0>
        "[2] Random Seed"=0
       )
      ))
    )) {isSetParameters=true}
   /SetVariables(Particles.MainColor:LinearColor=&MultiplyLinearColor(
     Linear Color=$Particles.MainColor
     Scale Factor=&MakeVector4_FromFloat(Float=&MakeCustomFloatFromBool(
       Bool=&Boolean_AndOperation(
        Bool1=&Boolean_OrOperation(Bool1=$User.bisADS, Bool2=$User.bIsADSOther)
        Bool2=&SetBoolByIntComparison(A=$Particles.PartitionParticles.Partition, B=7, Comparison Type=::NewEnumerator4)
       )
       True Float=0
       False Float=1
      ))
    )) {isSetParameters=true}
   /SetVariables(Particles.MainColor:LinearColor=&MultiplyLinearColor(
     Linear Color=$Particles.MainColor
     Scale Factor=&MakeVector4_FromFloat(Float=&MakeCustomFloatFromBool(
       Bool=&Boolean_AndOperation(
        Bool1=$User.bisADS
        Bool2=&SetBoolByIntComparison(A=$Particles.PartitionParticles.Partition, B=2, Comparison Type=::NewEnumerator7)
       )
       True Float=0
       False Float=1
      ))
    )) {enabled=false, isSetParameters=true}
   /SetVariables(
    Particles.SoftGlowCameraOffset:NiagaraFloat=&SelectFloatFromArray(
     Direct Array Index=$Particles.PartitionParticles.Partition
     Float Selection Array=DataInterface() {propertyValues {FloatData [11,11,5,2,2,2,2,2,2,2]}}
    )
    Particles.SoftGlowColor:LinearColor=&MultiplyLinearColor(
     Linear Color=$StackContext.MainColor
     Scale Factor=&MakeFloat4FromVectorAndFloat(
      "Vector (RGB)"=&VectorFromFloat(Value=&SelectFloatFromArray(
        Direct Array Index=$Particles.PartitionParticles.Partition
        Float Selection Array=DataInterface() {propertyValues {FloatData [2,2,1,5,5,0.3,0.3,3,3,3]}}
       ))
      "Float (Alpha)"=0.5
     )
    )
    Particles.SoftGlowSize:Vector2f=&Multiply_Vector2D_ByFloat(
     Float=&Multiply_Float(
      A=&MakeCustomFloatFromBool(
       Bool=$User.bisHeadHidden
       True Float=&MakeCustomFloatFromBool(
        Bool=&SetBoolByIntComparison(A=$Particles.PartitionParticles.Partition, B=7, Comparison Type=::NewEnumerator2)
        True Float=0
        False Float=1
       )
       False Float=1
      )
      B=&MakeCustomFloatFromBool(
       Bool=$User.bisFootHidden
       True Float=&MakeCustomFloatFromBool(
        Bool=&Boolean_OrOperation(
         Bool1=&SetBoolByIntComparison(A=$Particles.PartitionParticles.Partition, B=3, Comparison Type=::NewEnumerator4)
         Bool2=&SetBoolByIntComparison(A=$Particles.PartitionParticles.Partition, B=4, Comparison Type=::NewEnumerator4)
        )
        True Float=0
        False Float=1
       )
       False Float=1
      )
     )
     Vector2D=&Vector2DFromFloat(Value=&SelectFloatFromArray(
       Direct Array Index=$Particles.PartitionParticles.Partition
       Float Selection Array=DataInterface() {propertyValues {FloatData [20,20,16,9,9,33,33,9,9,15]}}
      ))
    )
   ) {isSetParameters=true}
   /SetVariables(
    Particles.FakeLightColor:LinearColor=&MultiplyLinearColor(
     Linear Color=$StackContext.MainColor
     Scale Factor=&MakeFloat4FromVectorAndFloat(
      "Vector (RGB)"=&VectorFromFloat(Value=&SelectFloatFromArray(
        Direct Array Index=$Particles.PartitionParticles.Partition
        Float Selection Array=DataInterface() {propertyValues {FloatData [3,3,3,8,8,3,3,8,8,8]}}
       ))
      "Float (Alpha)"=0.25
     )
    )
    Particles.FakeLightSize:Vector2f=&Multiply_Vector2D_ByFloat(
     Float=&Multiply_Float(
      A=&MakeCustomFloatFromBool(
       Bool=$User.bisHeadHidden
       True Float=&MakeCustomFloatFromBool(
        Bool=&SetBoolByIntComparison(A=$Particles.PartitionParticles.Partition, B=7, Comparison Type=::NewEnumerator2)
        True Float=0
        False Float=1
       )
       False Float=1
      )
      B=&MakeCustomFloatFromBool(
       Bool=$User.bisFootHidden
       True Float=&MakeCustomFloatFromBool(
        Bool=&Boolean_OrOperation(
         Bool1=&SetBoolByIntComparison(A=$Particles.PartitionParticles.Partition, B=3, Comparison Type=::NewEnumerator4)
         Bool2=&SetBoolByIntComparison(A=$Particles.PartitionParticles.Partition, B=4, Comparison Type=::NewEnumerator4)
        )
        True Float=0
        False Float=1
       )
       False Float=1
      )
     )
     Vector2D=&Vector2DFromFloat(Value=&SelectFloatFromArray(
       Direct Array Index=$Particles.PartitionParticles.Partition
       Float Selection Array=DataInterface() {propertyValues {FloatData [20,20,15,11,11,33,33,11,11,15]}}
      ))
    )
   ) {isSetParameters=true}
   /SetVariables(
    Particles.CapsuleGlowColor:LinearColor=&MultiplyLinearColor(
     Linear Color=$StackContext.MainColor
     Scale Factor=&MakeFloat4FromVectorAndFloat(
      "Vector (RGB)"=&VectorFromFloat(Value=&SelectFloatFromArray(
        Direct Array Index=$Particles.PartitionParticles.Partition
        Float Selection Array=DataInterface() {propertyValues {FloatData [2,2,2,3,3,3,3,3,3,3]}}
       ))
      "Float (Alpha)"=1
     )
    )
    Particles.CapsuleSpriteSize:Vector2f=&Multiply_Vector2D_ByFloat(
     Float=&Multiply_Float(
      A=&MakeCustomFloatFromBool(
       Bool=$User.bisHeadHidden
       True Float=&MakeCustomFloatFromBool(
        Bool=&SetBoolByIntComparison(A=$Particles.PartitionParticles.Partition, B=7, Comparison Type=::NewEnumerator2)
        True Float=0
        False Float=1
       )
       False Float=1
      )
      B=&MakeCustomFloatFromBool(
       Bool=$User.bisFootHidden
       True Float=&MakeCustomFloatFromBool(
        Bool=&Boolean_OrOperation(
         Bool1=&SetBoolByIntComparison(A=$Particles.PartitionParticles.Partition, B=3, Comparison Type=::NewEnumerator4)
         Bool2=&SetBoolByIntComparison(A=$Particles.PartitionParticles.Partition, B=4, Comparison Type=::NewEnumerator4)
        )
        True Float=0
        False Float=1
       )
       False Float=1
      )
     )
     Vector2D=&Vector2DFromFloat(Value=&SelectFloatFromArray(
       Direct Array Index=$Particles.PartitionParticles.Partition
       Float Selection Array=DataInterface() {propertyValues {FloatData [11,11,8,4,4,8,8,3,3,6]}}
      ))
    )
   ) {isSetParameters=true}
   /SetVariables(
    Particles.LightColor=&MultiplyLinearColor(
     Linear Color=$StackContext.MainColor
     Scale Factor=&MakeFloat4FromVectorAndFloat(
      "Vector (RGB)"=&VectorFromFloat(Value=&SelectFloatFromArray(
        Direct Array Index=$Particles.PartitionParticles.Partition
        Float Selection Array=DataInterface() {propertyValues {FloatData [1,1,0,0,0,1,1,1,1,1]}}
       ))
      "Float (Alpha)"=33
     )
    )
    Particles.LightRadius=&SelectFloatFromArray(
     Direct Array Index=$Particles.PartitionParticles.Partition
     Float Selection Array=DataInterface() {propertyValues {FloatData [33,33,0,0,0,0,0,0,0,0]}}
    )
   ) {isSetParameters=true}
   &SkeletalMeshLocation(
    Bone / Socket Index=&SelectIntFromArray(
     Direct Array Index=$Particles.PartitionParticles.Partition
     Int Selection Array=DataInterface() {propertyValues {IntData [0,0,0,2,2,0,0,6,6,4]}}
    )
    Bone Sampling Mode=::NewEnumerator4
    MaskSpawnGroup=false
    Sampled Position Offset=&SelectVectorFromArray(
     Direct Array Index=$Particles.PartitionParticles.Partition
     Vector Selection Array=DataInterface() {propertyValues {FloatData [<7.5 12 9>, <7.5 12 -9>, <4.5 14 0>, <3 6 -3>, <3 6 4.5>, <13 -12 -3.75>, <13 -12 3.75>, <4.25 8 -2.5>, <4.25 8 2.5>, <15 1.5 2.4>]}}
    )
    Skeletal Mesh=$System.Skeletal Mesh
   )
   /SetVariables(Particles.CapsuleSpriteAlignment:Vector3f=&RotateVectorByQuaternion(
     Vector To Rotate=&SelectVectorFromArray(
      Direct Array Index=$Particles.PartitionParticles.Partition
      Vector Selection Array=DataInterface() {propertyValues {FloatData [<0.5 -0.375 1>, <-0.5 0.375 1>, <1 0 0>, <-0.3 0.5 1>, <0.2 -0.5 1>, <1 0 -0.3>, <1 0 0.3>, <0 0.25 1>, <0 -0.25 1>, <0 0 0>]}}
     )
     Quaternion=$Particles.SkeletalMeshLocation.SampledRotation
    )) {isSetParameters=true}
   /SetVariables(Particles.CapsuleSpriteDynamic:Vector4f=&MakeFloat4FromVectorAndFloat(
     "Vector (RGB)"=&NormalizeVector(Vector To Normalize=$Particles.CapsuleSpriteAlignment, Safe Normalize=true, Safe Fallback Threshold=0, Safe Fallback Vector=<0 0 0.001>)
     "Float (Alpha)"=&SelectFloatFromArray(
      Direct Array Index=$Particles.PartitionParticles.Partition
      Float Selection Array=DataInterface() {propertyValues {FloatData [15,15,15,7,7,17,17,5,5,0]}}
     )
    )) {isSetParameters=true}
   /SetVariables(Particles.MeshOrientation=&OrientMeshToVector(
     Look At Direction=&NormalizeVector(Vector To Normalize=$Particles.CapsuleSpriteAlignment, Safe Normalize=true, Safe Fallback Threshold=0, Safe Fallback Vector=<0 0 0.001>)
     Reference Axis Vector=<1 0 0>
    )) {enabled=false, isSetParameters=true}
  ]
  rendererClasses [&NiagaraSpriteRendererProperties, &NiagaraLightRendererProperties]
  renderers [
   {rendererClass=&NiagaraSpriteRendererProperties}
   {rendererClass=&NiagaraSpriteRendererProperties, rendererIndex=1}
   {rendererClass=&NiagaraSpriteRendererProperties, rendererIndex=2}
   {rendererClass=&NiagaraLightRendererProperties, rendererIndex=3}
  ]
 }
}
summary {
 systemName=NS_Character_GreatPoolSwim_Trunks_SmokyFlames
 emitters [
  {emitterName=Flames_Spline_Mesh, rendererClasses [&NiagaraMeshRendererProperties]}
  {emitterName=Embers, rendererClasses [&NiagaraSpriteRendererProperties]}
  {emitterName=FlameSprites, rendererClasses [&NiagaraSpriteRendererProperties]}
  {
   emitterName=GlowsAndLights
   rendererClasses [&NiagaraSpriteRendererProperties, &NiagaraLightRendererProperties]
  }
 ]
 userVariables [
  User.bisADS {type=NiagaraBool}
  User.bIsADSOther {type=NiagaraBool}
  User.bisFootHidden {type=NiagaraBool}
  User.bIsFrontEnd {type=NiagaraBool}
  User.bIsFrontEndPreview {type=NiagaraBool}
  User.bisHeadHidden {type=NiagaraBool}
  User.bShowBackFX {type=NiagaraBool, defaultValue=true}
  User.bShowEyeFX {type=NiagaraBool, defaultValue=true}
  User.DebugReactivity {type=NiagaraBool}
  User.KillReactivityCount {type=NiagaraInt32}
 ]
}
